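# Physics asset for the Durrburger male medium player body, covering only the head and its dangling tongue for secondary motion.
# A head capsule anchors a chain of three tongue bodies: a small zero-length capsule at the root, then two thin, flat boxes.
# Each constraint zeroes twist and swing 1 limits and narrows swing 2 slightly toward the tip, so the tongue bends on a single axis like a hinged flap.
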PhysicsAsset /BRCosmetics/Characters/Player/Male/Medium/Bodies/M_MED_Durrburger/Meshes/M_MED_Durrburger_Physics
bodies {
 dyn_tongue_01 [{shapeType=Capsule, length=0, radius=3, center=<0.1 0 0>}]
 dyn_tongue_02 [{shapeType=Box, extentX=8, extentY=13.5, extentZ=3, center=<3.7746 0 0.0171>, rotation=<rot -5 0 0>}]
 dyn_tongue_03 [{shapeType=Box, extentX=9, extentY=15, extentZ=2, center=<4.9511 0 -0.4475>, rotation=<rot -5 0 0>}]
 head [{shapeType=Capsule, length=5.8074, radius=15.1385, center=<5.4625 0.1981 -0.4239>, rotation=<rot 90 0 0>}]
}
constraints [
 {bone1Name=dyn_tongue_01, bone2Name=head, swing1LimitDegrees=0, swing2LimitDegrees=40, twistLimitDegrees=0}
 {bone1Name=dyn_tongue_02, bone2Name=dyn_tongue_01, swing1LimitDegrees=0, swing2LimitDegrees=37.5, twistLimitDegrees=0}
 {bone1Name=dyn_tongue_03, bone2Name=dyn_tongue_02, swing1LimitDegrees=0, swing2LimitDegrees=35, twistLimitDegrees=0}
]
