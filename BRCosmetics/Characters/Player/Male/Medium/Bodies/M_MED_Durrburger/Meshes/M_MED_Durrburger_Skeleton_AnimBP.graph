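# Post-style secondary-motion AnimBlueprint for the Durrburger player body, adding physics jiggle to a pose copied from the attached parent mesh.
# Chained AnimDynamics nodes simulate tongue, pouch, condiments, belt, bun, cheese and eyes, then a RigidBody node runs the Durrburger physics asset.
# Pawn 2D speed drives the bun's clamped, interpolated alpha.
# Alternate tongue and holster nodes are disconnected; RBAN Off is set each update but unread here.
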AnimBlueprint /BRCosmetics/Characters/Player/Male/Medium/Bodies/M_MED_Durrburger/Meshes/M_MED_Durrburger_Skeleton_AnimBP
target_skeleton=&/M_MED_Durrburger_Skeleton
graphs [
 AnimGraph {nodes [
   n1 = Anim Dynamics - Bone: dyn_tongue_03(Alpha=0, ExternalForce="0,0,0") {
    alpha=0.9
    angularSpringConstant=50
    bAngularSpring=true
    bGravityOverrideInSimSpace=true
    numSolverIterationsPostUpdate=4
    numSolverIterationsPreUpdate=16
    simulationSpace=Actor
    alphaBoolBlend {blendOption=Linear}
    boundBone {boneName=dyn_tongue_03}
    physicsBodyDefinitions [{
      sphereCollisionRadius=0
      boundBone {boneName=dyn_tongue_03}
      boxExtents=<3 15 5>
      constraintSetup {angularTargetAxis=AxisZ, twistAxis=AxisY, angularLimitsMax=<15 30 15>, angularLimitsMin=<-15 -30 -15>, angularTarget=<0 0 1>}
      localJointOffset=<0 0 0>
     }]
   }
   Anim Dynamics - Bone: dyn_tongue_04(ComponentPose=@n1, Alpha=0, ExternalForce="0,0,0") {
    alpha=0.9
    angularSpringConstant=1500
    bGravityOverrideInSimSpace=true
    numSolverIterationsPostUpdate=4
    numSolverIterationsPreUpdate=16
    simulationSpace=Actor
    alphaBoolBlend {blendOption=Linear}
    boundBone {boneName=dyn_tongue_04}
    physicsBodyDefinitions [{
      sphereCollisionRadius=0
      boundBone {boneName=dyn_tongue_04}
      boxExtents=<3 15 5>
      constraintSetup {angularTargetAxis=AxisZ, twistAxis=AxisY, angularLimitsMax=<15 30 15>, angularLimitsMin=<-15 -30 -15>, angularTarget=<0 0 1>}
      localJointOffset=<0 0 0>
     }]
   }
   Anim Dynamics - Bone: dyn_holster(
    ComponentPose=Reroute Node(InputPin:Component Space Pose Link Structure) -> (OutputPin:Component Space Pose Link Structure)
    Alpha=1
    ExternalForce="0,0,0"
   ) {
    bGravityOverrideInSimSpace=true
    gravityScale=0.1
    lODThreshold=1
    alphaBoolBlend {blendOption=Linear}
    alphaScaleBiasClamp {bClampResult=true, clampMax=0.33, clampMin=0.1, scale=0.0033}
    boundBone {boneName=dyn_holster}
    componentAppliedLinearAccClamp=<2500 2500 2500>
    componentLinearAccScale=<1 1 1>
    componentLinearVelScale=<1 1 1>
    physicsBodyDefinitions [{
      sphereCollisionRadius=0
      boundBone {boneName=dyn_holster}
      boxExtents=<5 5 5>
      constraintSetup {
       angularTargetAxis=AxisY
       twistAxis=AxisZ
       angularLimitsMax=<15 0 0>
       angularLimitsMin=<-15 0 0>
       angularTarget=<0 0 0>
       linearAxesMax=<0 0 2>
      }
      localJointOffset=<0 0 5>
     }]
   }
   n2 = Anim Dynamics - Bone: dyn_tongue_01(ComponentPose=Local To Component(LocalPose=Copy Pose From Mesh() {bUseAttachedParent=true}), Alpha=1, ExternalForce="0,0,0") {
    angularDampingOverride=0.9999
    angularSpringConstant=500
    bAngularSpring=true
    bGravityOverrideInSimSpace=true
    bOverrideAngularDamping=true
    simulationSpace=BoneRelative
    alphaBoolBlend {blendOption=Linear}
    boundBone {boneName=dyn_tongue_01}
    componentAppliedLinearAccClamp=<3000 3000 3000>
    componentLinearAccScale=<0.2 0.5 0.7>
    componentLinearVelScale=<0.2 0.5 0.7>
    physicsBodyDefinitions [{
      sphereCollisionRadius=0
      boundBone {boneName=dyn_tongue_01}
      boxExtents=<3 15 15>
      constraintSetup {angularLimitsMax=<0.2 30 20>, angularLimitsMin=<-0.2 -10 -20>, angularTarget=<0 0 0>}
      localJointOffset=<2 0 0>
     }]
    relativeSpaceBone {boneName=attach}
   }
   n3 = Anim Dynamics - Bone: dyn_tongue_02(ComponentPose=@n2, Alpha=1, ExternalForce="0,0,0") {
    angularDampingOverride=0.9999
    angularSpringConstant=500
    bAngularSpring=true
    bGravityOverrideInSimSpace=true
    bOverrideAngularDamping=true
    simulationSpace=BoneRelative
    alphaBoolBlend {blendOption=Linear}
    boundBone {boneName=dyn_tongue_02}
    componentAppliedLinearAccClamp=<3000 3000 3000>
    componentLinearAccScale=<0.2 0.5 0.7>
    componentLinearVelScale=<0.2 0.5 0.7>
    physicsBodyDefinitions [{
      sphereCollisionRadius=0
      boundBone {boneName=dyn_tongue_02}
      boxExtents=<3 15 15>
      constraintSetup {angularLimitsMax=<0.2 30 20>, angularLimitsMin=<-0.2 -10 -20>, angularTarget=<0 0 0>}
      localJointOffset=<2 0 0>
     }]
    relativeSpaceBone {boneName=attach}
   }
   n4 = Anim Dynamics - Bone: dyn_tongue_03(ComponentPose=@n3, Alpha=1, ExternalForce="0,0,0") {
    angularDampingOverride=0.9999
    angularSpringConstant=500
    bAngularSpring=true
    bGravityOverrideInSimSpace=true
    bOverrideAngularDamping=true
    simulationSpace=BoneRelative
    alphaBoolBlend {blendOption=Linear}
    boundBone {boneName=dyn_tongue_03}
    componentAppliedLinearAccClamp=<3000 3000 3000>
    componentLinearAccScale=<0.2 0.8 0.8>
    componentLinearVelScale=<0.2 0.8 0.8>
    physicsBodyDefinitions [{
      sphereCollisionRadius=0
      boundBone {boneName=dyn_tongue_03}
      boxExtents=<3 15 15>
      constraintSetup {angularLimitsMax=<0.2 20 20>, angularLimitsMin=<-0.2 -10 -20>, angularTarget=<0 0 0>}
      localJointOffset=<2 0 0>
     }]
    relativeSpaceBone {boneName=attach}
   }
   n5 = Anim Dynamics - Bone: dyn_tongue_04(ComponentPose=@n4, Alpha=1, ExternalForce="0,0,0") {
    angularDampingOverride=0.9999
    angularSpringConstant=500
    bAngularSpring=true
    bGravityOverrideInSimSpace=true
    bOverrideAngularDamping=true
    simulationSpace=BoneRelative
    alphaBoolBlend {blendOption=Linear}
    boundBone {boneName=dyn_tongue_04}
    componentAppliedLinearAccClamp=<3000 3000 3000>
    componentLinearAccScale=<0.2 0.8 0.8>
    componentLinearVelScale=<0.2 0.8 0.8>
    physicsBodyDefinitions [{
      sphereCollisionRadius=0
      boundBone {boneName=dyn_tongue_04}
      boxExtents=<3 15 15>
      constraintSetup {angularLimitsMax=<0.2 20 20>, angularLimitsMin=<-0.2 -10 -20>, angularTarget=<0 0 0>}
      localJointOffset=<2 0 0>
     }]
    relativeSpaceBone {boneName=attach}
   }
   n6 = Anim Dynamics - Bone: dyn_pouch_r(ComponentPose=@n5, Alpha=1, ExternalForce="0,0,0") {
    bGravityOverrideInSimSpace=true
    gravityScale=0.1
    lODThreshold=1
    alphaBoolBlend {blendOption=Linear}
    alphaScaleBiasClamp {bClampResult=true, clampMax=0.33, clampMin=0.1, scale=0.0033}
    boundBone {boneName=dyn_pouch_r}
    componentAppliedLinearAccClamp=<2500 2500 2500>
    componentLinearAccScale=<1 1 1>
    componentLinearVelScale=<1 1 1>
    physicsBodyDefinitions [{
      sphereCollisionRadius=0
      boundBone {boneName=dyn_pouch_r}
      boxExtents=<5 5 5>
      constraintSetup {angularTargetAxis=AxisY, twistAxis=AxisY, angularLimitsMax=<0 0 0>, angularLimitsMin=<0 0 0>, angularTarget=<0 0 0>, linearAxesMax=<0 0 2>}
      localJointOffset=<0 0 5>
     }]
   }
   n7 = Anim Dynamics - Bone: dyn_mustard(ComponentPose=@n6, Alpha=1, ExternalForce="0,0,0") {
    alpha=0.9
    angularDampingOverride=0.98
    angularSpringConstant=1500
    bAngularSpring=true
    bGravityOverrideInSimSpace=true
    bOverrideAngularDamping=true
    numSolverIterationsPostUpdate=4
    numSolverIterationsPreUpdate=16
    alphaBoolBlend {blendOption=Linear}
    boundBone {boneName=dyn_mustard}
    componentAppliedLinearAccClamp=<2500 2500 2500>
    componentLinearAccScale=<1 1 1>
    componentLinearVelScale=<1 1 1>
    physicsBodyDefinitions [{
      sphereCollisionRadius=0
      boundBone {boneName=dyn_mustard}
      boxExtents=<15 5 5>
      constraintSetup {angularTargetAxis=AxisZ, twistAxis=AxisY, angularLimitsMax=<15 15 15>, angularLimitsMin=<-15 -15 -15>, angularTarget=<0 0 1>}
      localJointOffset=<0 0 0>
     }]
   }
   n8 = Anim Dynamics - Bone: dyn_ketchup(ComponentPose=@n7, Alpha=1, ExternalForce="0,0,0") {
    alpha=0.9
    angularDampingOverride=0.98
    angularSpringConstant=1500
    bAngularSpring=true
    bGravityOverrideInSimSpace=true
    bOverrideAngularDamping=true
    numSolverIterationsPostUpdate=4
    numSolverIterationsPreUpdate=16
    alphaBoolBlend {blendOption=Linear}
    boundBone {boneName=dyn_ketchup}
    componentAppliedLinearAccClamp=<2500 2500 2500>
    componentLinearAccScale=<1 1 1>
    componentLinearVelScale=<1 1 1>
    physicsBodyDefinitions [{
      sphereCollisionRadius=0
      boundBone {boneName=dyn_ketchup}
      boxExtents=<15 5 5>
      constraintSetup {angularTargetAxis=AxisZ, twistAxis=AxisY, angularLimitsMax=<15 15 15>, angularLimitsMin=<-15 -15 -15>, angularTarget=<0 0 1>}
      localJointOffset=<0 0 0>
     }]
   }
   n9 = Anim Dynamics - Bone: dyn_belt(ComponentPose=@n8, Alpha=1, ExternalForce="0,0,0") {
    alpha=0.9
    angularDampingOverride=0.98
    angularSpringConstant=1500
    bGravityOverrideInSimSpace=true
    numSolverIterationsPostUpdate=4
    numSolverIterationsPreUpdate=16
    alphaBoolBlend {blendOption=Linear}
    boundBone {boneName=dyn_belt}
    componentAppliedLinearAccClamp=<2500 2500 2500>
    componentLinearAccScale=<1 1 1>
    componentLinearVelScale=<1 1 1>
    physicsBodyDefinitions [{
      sphereCollisionRadius=0
      boundBone {boneName=dyn_belt}
      boxExtents=<20 5 15>
      constraintSetup {angularTargetAxis=AxisZ, twistAxis=AxisZ, angularLimitsMax=<30 10 7>, angularLimitsMin=<-30 -10 -7>, angularTarget=<0 0 1>}
      localJointOffset=<0 1 0>
     }]
   }
   n10 = Anim Dynamics - Bone: dyn_bunTop(ComponentPose=@n9, Alpha=Get PawnSpeed2D(), ExternalForce="0,0,0") {
    angularDampingOverride=0.98
    angularSpringConstant=25000
    bAngularSpring=true
    bGravityOverrideInSimSpace=true
    bLinearSpring=true
    bOverrideAngularDamping=true
    bOverrideLinearDamping=true
    gravityScale=0.1
    lODThreshold=1
    linearDampingOverride=0.97
    linearSpringConstant=350
    alphaBoolBlend {blendOption=Linear}
    alphaScaleBiasClamp {bClampResult=true, bInterpResult=true, clampMax=0.1, interpSpeedDecreasing=2.5, interpSpeedIncreasing=2.5, outRange {max=0.1}}
    boundBone {boneName=dyn_bunTop}
    componentAppliedLinearAccClamp=<2500 2500 2500>
    componentLinearAccScale=<1 1 1>
    componentLinearVelScale=<1 1 1>
    physicsBodyDefinitions [{
      sphereCollisionRadius=0
      boundBone {boneName=dyn_bunTop}
      boxExtents=<50 50 50>
      constraintSetup {
       angularTargetAxis=AxisZ
       angularLimitsMax=<45 45 45>
       angularLimitsMin=<-45 -45 -45>
       angularTarget=<0 0 1>
       linearAxesMax=<0 0 10>
       linearAxesMin=<0 0 -10>
      }
      localJointOffset=<0 0 0>
     }]
   }
   n11 = Anim Dynamics - Bone: dyn_backCheese(ComponentPose=@n10, Alpha=1, ExternalForce="0,0,0") {
    alpha=0.9
    angularDampingOverride=0.98
    angularSpringConstant=1500
    bAngularSpring=true
    bGravityOverrideInSimSpace=true
    bOverrideAngularDamping=true
    numSolverIterationsPostUpdate=4
    numSolverIterationsPreUpdate=16
    alphaBoolBlend {blendOption=Linear}
    boundBone {boneName=dyn_backCheese}
    componentAppliedLinearAccClamp=<2500 2500 2500>
    componentLinearAccScale=<1 1 1>
    componentLinearVelScale=<1 1 1>
    physicsBodyDefinitions [{
      sphereCollisionRadius=0
      boundBone {boneName=dyn_backCheese}
      boxExtents=<15 5 5>
      constraintSetup {angularTargetAxis=AxisZ, twistAxis=AxisY, angularLimitsMax=<15 15 15>, angularLimitsMin=<-15 -15 -15>, angularTarget=<0 0 1>}
      localJointOffset=<0 0 0>
     }]
   }
   n12 = Anim Dynamics - Bone: dyn_frontCheese(ComponentPose=@n11, Alpha=1, ExternalForce="0,0,0") {
    alpha=0.9
    angularDampingOverride=0.98
    angularSpringConstant=1500
    bAngularSpring=true
    bGravityOverrideInSimSpace=true
    bOverrideAngularDamping=true
    numSolverIterationsPostUpdate=4
    numSolverIterationsPreUpdate=16
    alphaBoolBlend {blendOption=Linear}
    boundBone {boneName=dyn_frontCheese}
    componentAppliedLinearAccClamp=<2500 2500 2500>
    componentLinearAccScale=<1 1 1>
    componentLinearVelScale=<1 1 1>
    physicsBodyDefinitions [{
      sphereCollisionRadius=0
      boundBone {boneName=dyn_frontCheese}
      boxExtents=<15 5 5>
      constraintSetup {angularTargetAxis=AxisZ, twistAxis=AxisY, angularLimitsMax=<15 15 15>, angularLimitsMin=<-15 -15 -15>, angularTarget=<0 0 1>}
      localJointOffset=<0 0 0>
     }]
   }
   n13 = Anim Dynamics - Bone: dyn_smallCheese(ComponentPose=@n12, Alpha=1, ExternalForce="0,0,0") {
    alpha=0.9
    angularDampingOverride=0.98
    angularSpringConstant=1500
    bAngularSpring=true
    bGravityOverrideInSimSpace=true
    bOverrideAngularDamping=true
    numSolverIterationsPostUpdate=4
    numSolverIterationsPreUpdate=16
    alphaBoolBlend {blendOption=Linear}
    boundBone {boneName=dyn_smallCheese}
    componentAppliedLinearAccClamp=<2500 2500 2500>
    componentLinearAccScale=<1 1 1>
    componentLinearVelScale=<1 1 1>
    physicsBodyDefinitions [{
      sphereCollisionRadius=0
      boundBone {boneName=dyn_smallCheese}
      boxExtents=<15 5 5>
      constraintSetup {angularTargetAxis=AxisZ, twistAxis=AxisY, angularLimitsMax=<15 15 15>, angularLimitsMin=<-15 -15 -15>, angularTarget=<0 0 1>}
      localJointOffset=<0 0 0>
     }]
   }
   n14 = Anim Dynamics - Bone: dyn_eye_r(ComponentPose=@n13, Alpha=1, ExternalForce="0,0,0") {
    angularDampingOverride=0.1
    angularSpringConstant=10
    bAngularSpring=true
    bGravityOverrideInSimSpace=true
    bOverrideAngularDamping=true
    bOverrideLinearDamping=true
    bUsePlanarLimit=false
    gravityScale=0.33
    linearDampingOverride=0.5
    numSolverIterationsPostUpdate=4
    numSolverIterationsPreUpdate=16
    alphaBoolBlend {blendOption=Linear}
    boundBone {boneName=dyn_eye_r}
    physicsBodyDefinitions [{
      sphereCollisionRadius=0
      boundBone {boneName=dyn_eye_r}
      boxExtents=<5 5 5>
      constraintSetup {
       angularTargetAxis=AxisY
       coneAngle=45
       angularLimitsMax=<0.001 20 20>
       angularLimitsMin=<-0.001 -20 -10>
       angularTarget=<0 0 0>
       linearAxesMax=<0 0 0.5>
       linearAxesMin=<0 0 -0.5>
      }
      localJointOffset=<0 0 0>
     }]
   }
   n15 = Anim Dynamics - Bone: dyn_eye_l(ComponentPose=@n14, Alpha=1, ExternalForce="0,0,0") {
    angularDampingOverride=0.1
    angularSpringConstant=10
    bAngularSpring=true
    bGravityOverrideInSimSpace=true
    bOverrideAngularDamping=true
    bOverrideLinearDamping=true
    bUsePlanarLimit=false
    gravityScale=0.33
    linearDampingOverride=0.5
    numSolverIterationsPostUpdate=4
    numSolverIterationsPreUpdate=16
    alphaBoolBlend {blendOption=Linear}
    boundBone {boneName=dyn_eye_l}
    physicsBodyDefinitions [{
      sphereCollisionRadius=0
      boundBone {boneName=dyn_eye_l}
      boxExtents=<5 5 5>
      constraintSetup {
       angularTargetAxis=AxisY
       coneAngle=45
       angularLimitsMax=<0.001 20 20>
       angularLimitsMin=<-0.001 -20 -10>
       angularTarget=<0 0 0>
       linearAxesMax=<0 0 0.5>
       linearAxesMin=<0 0 -0.5>
      }
      localJointOffset=<0 0 0>
     }]
   }
   n16 = RigidBody(ComponentPose=@n15) {
    bOverrideWorldGravity=true
    overridePhysicsAsset=&/M_MED_Durrburger_Physics
    simulationSpace=BaseBoneSpace
    alphaBoolBlend {blendOption=Linear}
    baseBoneRef {boneName=root}
    componentAppliedLinearAccClamp=<500 500 500>
    componentLinearAccScale=<0.25 0.25 0.25>
    componentLinearVelScale=<0.25 0.25 0.25>
    overrideWorldGravity {z=-980}
   }
   Output Pose(Result=Component To Local(ComponentPose=@n16)) {layerGroup=None}
  ]}
 EventGraph {nodes [
   Event BlueprintUpdateAnimation()
   > Set RBAN Off(RBAN Off:double=SelectFloat(A=0, B=1, bPickA=Get bRigidBodyEnabled())) -> (Output_Get:double)
  ]}
]
variables [RBAN Off {type=number}]
